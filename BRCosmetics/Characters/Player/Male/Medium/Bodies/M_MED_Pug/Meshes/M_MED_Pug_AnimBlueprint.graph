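# Animation Blueprint for the M_MED_Pug player body cosmetic, which follows the attached parent mesh's pose and adds secondary motion to a shoe tag.
# Copy Pose From Mesh feeds two chained Anim Dynamics nodes on the shoe tag bones, simulated root-relative with angular springs, damping and limits.
# The result passes through the Pug's Control Rig to the output pose.
# The event graph holds only unconnected default nodes.
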AnimBlueprint /BRCosmetics/Characters/Player/Male/Medium/Bodies/M_MED_Pug/Meshes/M_MED_Pug_AnimBlueprint
target_skeleton=&/M_MED_Pug_Skeleton
graphs [
 AnimGraph {nodes [
   n1 = Anim Dynamics - Bone: dyn_shoe_tag_01(ComponentPose=Local To Component(LocalPose=Copy Pose From Mesh() {bUseAttachedParent=true}), Alpha=1, ExternalForce="0,0,0") {
    angularDampingOverride=0.999
    angularSpringConstant=30000
    bAngularSpring=true
    bGravityOverrideInSimSpace=true
    bOverrideAngularDamping=true
    gravityScale=0.35
    simulationSpace=RootRelative
    alphaBoolBlend {blendOption=Linear}
    boundBone {boneName=dyn_shoe_tag_01}
    physicsBodyDefinitions [{
      sphereCollisionRadius=0
      boundBone {boneName=dyn_shoe_tag_01}
      boxExtents=<5 20 10>
      constraintSetup {angularTargetAxis=AxisY, angularLimitsMax=<0 0 30>, angularLimitsMin=<0 0 -30>, angularTarget=<0 1 0>}
      localJointOffset=<0 5 0>
     }]
   }
   n2 = Anim Dynamics - Bone: dyn_shoe_tag_02(ComponentPose=@n1, Alpha=1, ExternalForce="0,0,0") {
    angularDampingOverride=0.999
    angularSpringConstant=10000
    bAngularSpring=true
    bGravityOverrideInSimSpace=true
    bOverrideAngularDamping=true
    simulationSpace=RootRelative
    alphaBoolBlend {blendOption=Linear}
    boundBone {boneName=dyn_shoe_tag_02}
    physicsBodyDefinitions [{
      sphereCollisionRadius=0
      boundBone {boneName=dyn_shoe_tag_02}
      boxExtents=<5 10 10>
      constraintSetup {angularTargetAxis=AxisY, angularLimitsMax=<0 0 30>, angularLimitsMin=<0 0 -30>, angularTarget=<0 1 0>}
      localJointOffset=<0 1 0>
     }]
   }
   n3 = Control Rig(Source=Component To Local(ComponentPose=@n2)) {
    alphaBoolBlend {blendOption=Linear}
    controlRigAssetReference {blueprintRigClass=&/M_MED_Pug_CtrlRig.M_MED_Pug_CtrlRig_C}
   }
   Output Pose(Result=@n3) {layerGroup=None}
  ]}
 EventGraph {nodes [
   TryGetPawnOwner()
   Event BlueprintUpdateAnimation()
  ]}
]
variables []
