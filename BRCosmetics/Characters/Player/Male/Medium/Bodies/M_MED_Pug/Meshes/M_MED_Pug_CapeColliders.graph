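# Physics Asset of collision shapes for the cape on the medium male Pug player body, as its name indicates.
# Capsules cover the pelvis, spine, neck and clavicles, with several overlapping on spine_02 and spine_05, and a sphere covers the head.
# Arm, leg and foot bodies exist but hold no shapes, and no constraints are defined, so it provides upper-body collision volumes rather than a jointed ragdoll.
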PhysicsAsset /BRCosmetics/Characters/Player/Male/Medium/Bodies/M_MED_Pug/Meshes/M_MED_Pug_CapeColliders
bodies {
 calf_l []
 calf_r []
 clavicle_l [{shapeType=Capsule, length=10.5136, radius=8.3927, center=<10.417 -0.7935 1.241>, rotation=<rot 80.9121 -2.4474 173.1624>}]
 clavicle_r [{shapeType=Capsule, length=11.0136, radius=7.8927, center=<-10.4769 0.796 -1.6156>, rotation=<rot 80.912 -2.4476 -6.8378>}]
 foot_l []
 foot_r []
 head [{shapeType=Sphere, radius=17.3125, center=<11.028 -0.174 0>}]
 lowerarm_l []
 lowerarm_r []
 neck_02 [{shapeType=Capsule, length=13.4375, radius=7.3175, center=<-0.2642 -0.3408 0>, rotation=<rot 90 -171.1193 -175.3978>}]
 pelvis [{shapeType=Capsule, length=8.6562, radius=13.5388, center=<-2.6215 0.3474 0>}]
 spine_01 [{shapeType=Capsule, length=11.2187, radius=12.4138, center=<-2.6215 -0.7156 0>}]
 spine_02 [
  {shapeType=Capsule, length=14.5937, radius=12.195, center=<-2.6215 0.4993 0.0001>}
  {shapeType=Capsule, length=29.875, radius=11.1013, center=<2.5024 0.4992 3.9825>, rotation=<rot 90 0.0049 0.0049>}
  {shapeType=Capsule, length=29.875, radius=11.1013, center=<2.4704 0.4991 -3.9825>, rotation=<rot 90 -0.0396 -0.0396>}
 ]
 spine_03 [{shapeType=Capsule, length=15.1562, radius=11.7575, center=<-0.6351 0.1364 0.0001>}]
 spine_04 [{shapeType=Capsule, length=18.9062, radius=10.8825, center=<-0.5619 -0.1462 0.0002>}]
 spine_05 [
  {shapeType=Capsule, length=25.3437, radius=9.695, center=<-2.8974 -0.7898 0>}
  {shapeType=Capsule, length=19.7812, radius=10.0625, center=<0.7875 0.1727 0>}
 ]
 thigh_l []
 thigh_r []
 upperarm_l []
 upperarm_r []
}
constraints []
