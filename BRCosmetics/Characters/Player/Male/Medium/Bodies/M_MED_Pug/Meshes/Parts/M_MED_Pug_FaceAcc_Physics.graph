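# Physics asset for the M_MED_Pug face accessory mesh, covering a left and a right hoodie drawstring.
# Each string is a chain of five thin capsule bodies hanging from spine_05, linked by constraints with narrow swing and tighter twist limits.
# Four wide capsules on spine_02 through spine_05 form the torso volume, joined by constraints with free swing and twist.
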PhysicsAsset /BRCosmetics/Characters/Player/Male/Medium/Bodies/M_MED_Pug/Meshes/Parts/M_MED_Pug_FaceAcc_Physics
bodies {
 dyn_hoodie_string_01_l [{shapeType=Capsule, length=5, radius=1, center=<3 0 0>, rotation=<rot 90 179.8813 179.8813>}]
 dyn_hoodie_string_01_r [{shapeType=Capsule, length=5, radius=1, center=<3 0 0>, rotation=<rot 90 0.0395 0.0394>}]
 dyn_hoodie_string_02_l [{shapeType=Capsule, length=5, radius=1, center=<3 0 0>, rotation=<rot 90 0.0099 0.0099>}]
 dyn_hoodie_string_02_r [{shapeType=Capsule, length=5, radius=1, center=<3 0 0>, rotation=<rot 90 0.0099 0.0099>}]
 dyn_hoodie_string_03_l [{shapeType=Capsule, length=5, radius=1, center=<3 0 0>, rotation=<rot 90 0.0099 0.0099>}]
 dyn_hoodie_string_03_r [{shapeType=Capsule, length=5, radius=1, center=<3 0 0>, rotation=<rot 90 0.0198 0.0198>}]
 dyn_hoodie_string_04_l [{shapeType=Capsule, length=5, radius=1, center=<3 0 0>, rotation=<rot 90 0.0198 0.0198>}]
 dyn_hoodie_string_04_r [{shapeType=Capsule, length=5, radius=1, center=<3 0 0>, rotation=<rot 90 -0.0396 -0.0396>}]
 dyn_hoodie_string_05_l [{shapeType=Capsule, length=5, radius=1, center=<3 0 0>, rotation=<rot 90 0.0198 0.0198>}]
 dyn_hoodie_string_05_r [{shapeType=Capsule, length=5, radius=1, center=<3 0 0>, rotation=<rot 90 -0.0396 -0.0396>}]
 spine_02 [{shapeType=Capsule, length=10.5, radius=13.25, center=<6.2936 0 0>}]
 spine_03 [{shapeType=Capsule, length=10.5, radius=13.25, center=<6.2936 0 0>}]
 spine_04 [{shapeType=Capsule, length=7, radius=13.75, center=<5.5399 0 0>}]
 spine_05 [{shapeType=Capsule, length=12, radius=13.25}]
}
constraints [
 {
  bone1Name=spine_05
  bone2Name=spine_04
  swing1LimitDegrees=45
  swing1Motion=Free
  swing2LimitDegrees=45
  swing2Motion=Free
  twistLimitDegrees=45
  twistMotion=Free
 }
 {
  bone1Name=spine_04
  bone2Name=spine_03
  swing1LimitDegrees=45
  swing1Motion=Free
  swing2LimitDegrees=45
  swing2Motion=Free
  twistLimitDegrees=45
  twistMotion=Free
 }
 {
  bone1Name=spine_03
  bone2Name=spine_02
  swing1LimitDegrees=45
  swing1Motion=Free
  swing2LimitDegrees=45
  swing2Motion=Free
  twistLimitDegrees=45
  twistMotion=Free
 }
 {bone1Name=dyn_hoodie_string_02_r, bone2Name=dyn_hoodie_string_01_r, swing1LimitDegrees=20, swing2LimitDegrees=20, twistLimitDegrees=5}
 {bone1Name=dyn_hoodie_string_03_r, bone2Name=dyn_hoodie_string_02_r, swing1LimitDegrees=20, swing2LimitDegrees=20, twistLimitDegrees=5}
 {bone1Name=dyn_hoodie_string_04_r, bone2Name=dyn_hoodie_string_03_r, swing1LimitDegrees=20, swing2LimitDegrees=20, twistLimitDegrees=5}
 {bone1Name=dyn_hoodie_string_05_r, bone2Name=dyn_hoodie_string_04_r, swing1LimitDegrees=20, swing2LimitDegrees=20, twistLimitDegrees=5}
 {bone1Name=dyn_hoodie_string_01_l, bone2Name=spine_05, swing1LimitDegrees=20, swing2LimitDegrees=20, twistLimitDegrees=5}
 {bone1Name=dyn_hoodie_string_02_l, bone2Name=dyn_hoodie_string_01_l, swing1LimitDegrees=20, swing2LimitDegrees=20, twistLimitDegrees=5}
 {bone1Name=dyn_hoodie_string_03_l, bone2Name=dyn_hoodie_string_02_l, swing1LimitDegrees=20, swing2LimitDegrees=20, twistLimitDegrees=5}
 {bone1Name=dyn_hoodie_string_04_l, bone2Name=dyn_hoodie_string_03_l, swing1LimitDegrees=20, swing2LimitDegrees=20, twistLimitDegrees=5}
 {bone1Name=dyn_hoodie_string_05_l, bone2Name=dyn_hoodie_string_04_l, swing1LimitDegrees=20, swing2LimitDegrees=20, twistLimitDegrees=5}
 {bone1Name=dyn_hoodie_string_01_r, bone2Name=spine_05, swing1LimitDegrees=20, swing2LimitDegrees=20, twistLimitDegrees=5}
]
