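# Animation Blueprint for the Pug character's face accessory part, adding secondary motion to hoodie strings, necklace and hood on top of the pose copied from the attached parent mesh.
# Weighted constraints tie string roots to spine and clavicle bones, then AnimDynamics and a spine_03-based RigidBody node simulate the dangling pieces.
# Each update maps head yaw delta from ref pose to an alpha blending an additive dyn_hood_counter_r corrective.
# Bone Driven Controller and Flap nodes sit unconnected.
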AnimBlueprint /BRCosmetics/Characters/Player/Male/Medium/Bodies/M_MED_Pug/Meshes/Parts/M_MED_Pug_FaceAcc_AnimBP
target_skeleton=&/M_MED_Pug_FaceAcc_Skeleton
graphs [
 AnimGraph {nodes [
   "None.rotateZ = head.rotateZ - Bone Driven Controller"!(ComponentPose:Component Space Pose Link Structure, Alpha:float=1) -> (Pose:Component Space Pose Link Structure) {
    is=title AnimGraphNode_BoneDrivenController
    bAffectTargetRotationZ=true
    sourceComponent=RotationZ
    alphaBoolBlend {blendOption=Linear}
    sourceBone {boneName=head}
   }
   n1 = "Constraint - dyn_hoodie_string_01_r (PPPP)"!(
    ComponentPose=Local To Component(LocalPose=Copy Pose From Mesh() {bUseAttachedParent=true})
    Alpha=1
    ConstraintWeights_0=0.2347
    ConstraintWeights_1:float=0.08
    ConstraintWeights_2:float=0.4167
    ConstraintWeights_3:float=0.2685
   ) -> (Pose) {
    alphaBoolBlend {blendOption=Linear}
    boneToModify {boneName=dyn_hoodie_string_01_r}
    constraintSetup [{targetBone {boneName=clavicle_r}}, {targetBone {boneName=spine_03}}, {targetBone {boneName=spine_04}}, {targetBone {boneName=spine_05}}]
    constraintWeights [1,1,1,1]
   }
   n2 = "Constraint - dyn_hoodie_string_01_l (PPPP)"!(
    ComponentPose=@n1
    Alpha=1
    ConstraintWeights_0=0.0796
    ConstraintWeights_1:float=0.4206
    ConstraintWeights_2:float=0.2716
    ConstraintWeights_3:float=0.2282
   ) -> (Pose) {
    alphaBoolBlend {blendOption=Linear}
    boneToModify {boneName=dyn_hoodie_string_01_l}
    constraintSetup [{targetBone {boneName=spine_03}}, {targetBone {boneName=spine_04}}, {targetBone {boneName=spine_05}}, {targetBone {boneName=clavicle_l}}]
    constraintWeights [1,1,1,1]
   }
   n3 = Anim Dynamics - Bone: dyn_hoodie_string_01_r(ComponentPose=@n2, Alpha=0.5, ExternalForce="0,0,0") {
    angularDampingOverride=1
    bGravityOverrideInSimSpace=true
    bOverrideAngularDamping=true
    gravityScale=0.35
    simulationSpace=RootRelative
    alphaBoolBlend {blendOption=Linear}
    boundBone {boneName=dyn_hoodie_string_01_r}
    physicsBodyDefinitions [{
      sphereCollisionRadius=0
      boundBone {boneName=dyn_hoodie_string_01_r}
      boxExtents=<5 10 10>
      constraintSetup {angularLimitsMax=<0 20 0>, angularLimitsMin=<0 -20 -40>, angularTarget=<0 0 0>}
      localJointOffset=<5 0 -0.01>
     }]
   }
   n4 = Anim Dynamics - Bone: dyn_hoodie_string_01_l(ComponentPose=@n3, Alpha=0.5, ExternalForce="0,0,0") {
    angularDampingOverride=1
    bGravityOverrideInSimSpace=true
    bOverrideAngularDamping=true
    gravityScale=0.35
    simulationSpace=RootRelative
    alphaBoolBlend {blendOption=Linear}
    boundBone {boneName=dyn_hoodie_string_01_l}
    physicsBodyDefinitions [{
      sphereCollisionRadius=0
      boundBone {boneName=dyn_hoodie_string_01_l}
      boxExtents=<5 10 10>
      constraintSetup {angularLimitsMax=<0 0 0>, angularLimitsMin=<0 0 -40>, angularTarget=<0 0 0>}
      localJointOffset=<5 0 -0.01>
     }]
   }
   n5 = RigidBody(ComponentPose=@n4) {
    bClampLinearTranslationLimitToRefPose=true
    overridePhysicsAsset=&/M_MED_Pug_FaceAcc_Physics
    simulationSpace=BaseBoneSpace
    alphaBoolBlend {blendOption=Linear}
    baseBoneRef {boneName=spine_03}
    simSpaceSettings {
     maxAngularAcceleration=1000
     maxAngularVelocity=3
     maxLinearAcceleration=700
     maxLinearVelocity=500
     velocityScaleZ=0
     worldAlpha=1
     externalLinearDragV=<0.2 0.2>
    }
   }
   n6 = Anim Dynamics - Bone: dyn_necklace_01(ComponentPose=@n5, Alpha=0.5, ExternalForce="0,0,0") {
    angularDampingOverride=1
    bGravityOverrideInSimSpace=true
    bOverrideAngularDamping=true
    gravityScale=0.35
    simulationSpace=RootRelative
    alphaBoolBlend {blendOption=Linear}
    boundBone {boneName=dyn_necklace_01}
    physicsBodyDefinitions [{
      sphereCollisionRadius=0
      boundBone {boneName=dyn_necklace_01}
      boxExtents=<5 10 10>
      constraintSetup {angularLimitsMax=<0 20 -3>, angularLimitsMin=<0 -20 -40>, angularTarget=<0 0 0>}
      localJointOffset=<5 0 0>
     }]
   }
   n7 = Anim Dynamics - Bone: dyn_necklace_02(ComponentPose=@n6, Alpha=1, ExternalForce="0,0,0") {
    angularDampingOverride=1
    bGravityOverrideInSimSpace=true
    bOverrideAngularDamping=true
    gravityScale=0.35
    simulationSpace=RootRelative
    alphaBoolBlend {blendOption=Linear}
    boundBone {boneName=dyn_necklace_02}
    physicsBodyDefinitions [{
      sphereCollisionRadius=0
      boundBone {boneName=dyn_necklace_02}
      boxExtents=<5 10 10>
      constraintSetup {angularLimitsMax=<0 0 -2>, angularLimitsMin=<0 0 -40>, angularTarget=<0 0 0>}
      localJointOffset=<5 0 0>
     }]
   }
   n8 = "Transform (Modify) Bone - Bone: dyn_hood_counter_r"(
    ComponentPose=@n7
    Alpha=Get head_corrective_alpha() -> (head_corrective_alpha:double)
    Translation="-0.000,-1.156,0.846"
    Rotation="0.000,-0.000,16.379"
    Scale="1,1,1"
   ) {
    rotationMode=BMM_Additive
    rotationSpace=BCS_BoneSpace
    translationMode=BMM_Additive
    translationSpace=BCS_BoneSpace
    alphaBoolBlend {blendOption=Linear}
    boneToModify {boneName=dyn_hood_counter_r}
    rotation {roll=16.379}
    translation {y=-1.156, z=0.846}
   }
   Output Pose(Result=Component To Local(ComponentPose=@n8)) {layerGroup=None}
   Flap Node - Bone: dyn_hood_counter_r!(ComponentPose:Component Space Pose Link Structure, Alpha:float=1) -> (Pose:Component Space Pose Link Structure) {
    is=title FortAnimGraphNode_Flap
    targetComponent=RotationZ
    alphaBoolBlend {blendOption=Linear}
    sourceDrives [{
      bUseRange=false
      drivingCurve=null
      multiplier=1
      rangeMax=1
      rangeMin=-1
      remappedMax=1
      remappedMin=0
      sourceComponent=RotationZ
      useQuaternion=true
      sourceBone {boneName=head}
     }]
    targetBone {boneName=dyn_hood_counter_r}
   }
  ]}
 EventGraph {nodes [
   Event BlueprintUpdateAnimation()
   n1 = BreakTransform(InTransform=GetDeltaTransformFromRefPose(self=GetPartSkeletalMeshComponent(), BoneName=head, BaseName=None))
   > Set head_corrective_alpha(head_corrective_alpha:double=MapRangeClamped(Value=BreakRotator(InRot=@n1.Rotation).Yaw, InRangeA=0, InRangeB=-30)) -> (Output_Get:double)
   TryGetPawnOwner()
  ]}
]
variables [head_corrective_alpha {type=number}]
